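# Physics asset for the cloak mesh of the AllKnowing medium male player body, defining secondary motion for cloak, collar and ropes.
# Capsules on spine, neck, head, clavicles and upper arms form the body shape.
# Front and back cloak chains and six collar flaps hang from MSTR anchor bones on the upper spine with angle-limited joints, mostly without twist.
# Ropes branch from cloak segments, spine_04 or the left clavicle.
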PhysicsAsset /BRCosmetics/Characters/Player/Male/Medium/Bodies/M_MED_AllKnowing/Meshes/M_MED_AllKnowing_cloak_Physics
bodies {
 clavicle_l [{
   shapeType=Capsule
   length=13.2608
   radius=9.517
   shapeName=clavicle_l_capsule
   center=<7.7595 0 0.0564>
   rotation=<rot 89.5833 -179.9676 -174.1287>
  }]
 clavicle_r [{
   shapeType=Capsule
   length=12.7998
   radius=9.978
   shapeName=clavicle_l_capsule
   center=<-7.7595 0 -0.0564>
   rotation=<rot 89.5833 -179.9676 5.8713>
  }]
 dyn_cloakBack_c_01 [{shapeType=Capsule, length=22.2594, radius=6.4563, shapeName=dyn_cloakBack_c_01_capsule, center=<5.1242 4.4 0>}]
 dyn_cloakBack_c_02 [{
   shapeType=Capsule
   length=12.6021
   radius=5.1083
   shapeName=dyn_cloakBack_c_02_capsule
   center=<8.4359 4.1846 0>
   rotation=<rot 90 179.5524 -177.296>
  }]
 dyn_cloakBack_c_MSTR [{shapeType=Capsule, length=1, radius=1, shapeName=dyn_cloakBack_c_MSTR_capsule}]
 dyn_cloakFront_c_01 [{shapeType=Capsule, length=13.9678, radius=5.9317, shapeName=dyn_cloakFront_c_01_capsule, center=<1.9478 4.4 0>}]
 dyn_cloakFront_c_02 [{shapeType=Capsule, length=13.9678, radius=5.9317, shapeName=dyn_cloakFront_c_02_capsule, center=<1.9478 4.4 0>, rotation=<rot 90 0 0>}]
 dyn_cloakFront_c_MSTR [{shapeType=Capsule, length=1, radius=1, shapeName=dyn_cloakFront_c_MSTR_capsule}]
 dyn_collarBack_01 [{shapeType=Capsule, length=13.9536, radius=2.3979, shapeName=dyn_collarBack_01_capsule, center=<6.4055 1.4 0>}]
 dyn_collarBack_MSTR [{shapeType=Capsule, length=0.505, radius=0.505, shapeName=dyn_collarBack_MSTR_capsule, rotation=<rot 0 -90 0>}]
 dyn_collarBack_l_01 [{shapeType=Capsule, length=12.2969, radius=1.7418, shapeName=dyn_collar_l_01_capsule, center=<5.6952 1.1 0>}]
 dyn_collarBack_l_MSTR [{shapeType=Capsule, length=0.505, radius=0.505, shapeName=dyn_collarBack_l_MSTR_capsule, rotation=<rot 0 -90 0>}]
 dyn_collarBack_r_01 [{shapeType=Capsule, length=9.5648, radius=2.4686, shapeName=dyn_collar_r_01_capsule, center=<5.6952 2 0>}]
 dyn_collarBack_r_MSTR [{shapeType=Capsule, length=0.505, radius=0.505, shapeName=dyn_collarBack_r_MSTR_capsule, rotation=<rot 0 -90 0>}]
 dyn_collarFront_01 [{shapeType=Capsule, length=10.5414, radius=1.4921, shapeName=dyn_collarFront_01_capsule, center=<8.5309 0 0>}]
 dyn_collarFront_MSTR [{shapeType=Capsule, length=0.505, radius=0.505, shapeName=dyn_collarFront_MSTR_capsule, rotation=<rot 0 -90 0>}]
 dyn_collar_l_01 [{shapeType=Capsule, length=10.2808, radius=1.7527, shapeName=dyn_collar_l_01_capsule, center=<10.1228 1 0>, rotation=<rot 8.1433 0 0>}]
 dyn_collar_l_MSTR [{shapeType=Capsule, length=0.505, radius=0.505, shapeName=dyn_collar_l_MSTR_capsule, rotation=<rot 0 -90 0>}]
 dyn_collar_r_01 [{shapeType=Capsule, length=9.5648, radius=2.4686, shapeName=dyn_collar_r_01_capsule, center=<8.0182 0 0>}]
 dyn_collar_r_MSTR [{shapeType=Capsule, length=0.505, radius=0.505, shapeName=dyn_collar_r_MSTR_capsule, rotation=<rot 0 -90 0>}]
 dyn_ropeBack_A_01 [{
   shapeType=Capsule
   length=7.8026
   radius=2.7106
   shapeName=dyn_ropeBack_D_MSTR_capsule
   center=<4.3009 1.8592 0>
   rotation=<rot 90 107.1027 129.9222>
  }]
 dyn_ropeBack_A_MSTR [{shapeType=Capsule, length=0.5, radius=0.5, shapeName=dyn_ropeBack_A_MSTR_capsule}]
 dyn_ropeBack_B_01 [{
   shapeType=Capsule
   length=4.6106
   radius=2.4757
   shapeName=dyn_ropeBack_D_MSTR_capsule
   center=<3.6191 0.8 0>
   rotation=<rot 81.7622 -0.0001 15.301>
  }]
 dyn_ropeBack_B_MSTR [{shapeType=Capsule, length=0.5, radius=0.5, shapeName=dyn_ropeBack_B_MSTR_capsule}]
 dyn_ropeBack_C_MSTR [{
   shapeType=Capsule
   length=4.7305
   radius=2.0144
   shapeName=dyn_ropeBack_C_MSTR_capsule
   center=<2.8 0 0>
   rotation=<rot 90 -152.6261 -143.2542>
  }]
 dyn_ropeBack_D_MSTR [{shapeType=Capsule, length=5.563, radius=1.5233, shapeName=dyn_ropeBack_D_MSTR_capsule, center=<3.7859 0 0>, rotation=<rot 81.7622 0 0>}]
 dyn_ropeBack_E_MSTR [{
   shapeType=Capsule
   length=5.6678
   radius=1.7877
   shapeName=dyn_ropeBack_C_MSTR_capsule
   center=<3.5736 1.1 0>
   rotation=<rot 90 180 -170.6281>
  }]
 dyn_ropeFront_A_MSTR [{shapeType=Capsule, length=5.9444, radius=1.2763, shapeName=dyn_ropeFront_A_MSTR_capsule, center=<3.4624 0 0>, rotation=<rot 90 0 0>}]
 dyn_ropeFront_B_MSTR [{shapeType=Capsule, length=6.1343, radius=1.2763, shapeName=dyn_ropeFront_B_MSTR_capsule, center=<3.102 0 0>, rotation=<rot 90 0 0>}]
 dyn_ropeFront_C_01 [{
   shapeType=Capsule
   length=5.5635
   radius=2.0269
   shapeName=dyn_ropeFront_C_MSTR_capsule
   center=<3.8782 2.1451 0>
   rotation=<rot 90 -7.125 6.9955>
  }]
 dyn_ropeFront_C_MSTR [{shapeType=Capsule, length=0.5, radius=0.5, shapeName=dyn_ropeFront_C_MSTR_capsule}]
 dyn_ropeFront_D_01 [{
   shapeType=Capsule
   length=5.0186
   radius=2.0676
   shapeName=dyn_ropeBack_D_MSTR_capsule
   center=<3.6651 2.3 0>
   rotation=<rot 81.7622 -0.0009 36.4006>
  }]
 dyn_ropeFront_D_MSTR [{shapeType=Capsule, length=0.5, radius=0.5, shapeName=dyn_ropeFront_D_MSTR_capsule}]
 head [
  {shapeType=Capsule, length=12.0824, radius=7.2058, shapeName=head_capsule, center=<5.7427 0.3 0.7518>, rotation=<rot 0 -87.9256 90>}
  {shapeType=Capsule, length=14.7103, radius=5.8186, shapeName=head_capsule, center=<4.3194 0 -4.332>, rotation=<rot 81.4478 -0.0002 0>}
  {
   shapeType=Capsule
   length=20.693
   radius=3.5404
   shapeName=head_capsule
   center=<2.6202 1.6408 8.4509>
   rotation=<rot -58.1683 -0.0004 10.6185>
  }
 ]
 neck_02 [{shapeType=Capsule, length=1.5229, radius=7.9676, shapeName=neck_02_capsule, rotation=<rot 0 -90 0>}]
 spine_02 [{shapeType=Capsule, length=19.6352, radius=12.5396, shapeName=spine_02_capsule, center=<2.983 -0.3127 -0.1595>}]
 spine_03 [{shapeType=Capsule, length=8.4772, radius=13.7253, shapeName=spine_03_capsule, center=<2.1195 -0.6049 0>}]
 spine_04 [{shapeType=Capsule, length=16.3009, radius=13.3041, shapeName=spine_04_capsule, center=<2.1003 -0.4418 0>}]
 spine_05 [{
   shapeType=Capsule
   length=9.0156
   radius=5.1179
   shapeName=spine_05_capsule
   center=<-4.7138 10.0492 -6.9237>
   rotation=<rot 59.3586 -0.0001 15.4789>
  }]
 upperarm_l [{shapeType=Capsule, length=14.8111, radius=8.6816, shapeName=upperarm_l_capsule, center=<7.4387 -1.1381 -0.9438>, rotation=<rot 90 0 0>}]
 upperarm_r [{shapeType=Capsule, length=22.4028, radius=9.3686, shapeName=upperarm_l_capsule, center=<-7.4387 0 0>, rotation=<rot 90 0 180>}]
}
constraints [
 {
  bone1Name=spine_05
  bone2Name=spine_04
  swing1LimitDegrees=45
  swing1Motion=Free
  swing2LimitDegrees=45
  swing2Motion=Free
  twistLimitDegrees=45
  twistMotion=Free
 }
 {
  bone1Name=clavicle_l
  bone2Name=spine_05
  swing1LimitDegrees=45
  swing1Motion=Free
  swing2LimitDegrees=45
  swing2Motion=Free
  twistLimitDegrees=45
  twistMotion=Free
 }
 {
  bone1Name=spine_04
  bone2Name=spine_03
  swing1LimitDegrees=45
  swing1Motion=Free
  swing2LimitDegrees=45
  swing2Motion=Free
  twistLimitDegrees=45
  twistMotion=Free
 }
 {
  bone1Name=clavicle_r
  bone2Name=spine_05
  swing1LimitDegrees=45
  swing1Motion=Free
  swing2LimitDegrees=45
  swing2Motion=Free
  twistLimitDegrees=45
  twistMotion=Free
 }
 {bone1Name=dyn_collarBack_01, bone2Name=dyn_collarBack_MSTR, swing1LimitDegrees=70, swing2LimitDegrees=20, twistLimitDegrees=20}
 {bone1Name=dyn_collarFront_01, bone2Name=dyn_collarFront_MSTR, swing1LimitDegrees=30, swing2LimitDegrees=20, twistLimitDegrees=0}
 {bone1Name=dyn_collar_l_01, bone2Name=dyn_collar_l_MSTR, swing1LimitDegrees=30, swing2LimitDegrees=20, twistLimitDegrees=0}
 {bone1Name=dyn_collar_r_01, bone2Name=dyn_collar_r_MSTR, swing1LimitDegrees=30, swing2LimitDegrees=20, twistLimitDegrees=0}
 {
  bone1Name=dyn_collar_r_MSTR
  bone2Name=spine_05
  swing1LimitDegrees=45
  swing1Motion=Free
  swing2LimitDegrees=45
  swing2Motion=Free
  twistLimitDegrees=45
  twistMotion=Free
 }
 {
  bone1Name=dyn_collar_l_MSTR
  bone2Name=spine_05
  swing1LimitDegrees=45
  swing1Motion=Free
  swing2LimitDegrees=45
  swing2Motion=Free
  twistLimitDegrees=45
  twistMotion=Free
 }
 {
  bone1Name=dyn_collarFront_MSTR
  bone2Name=spine_05
  swing1LimitDegrees=45
  swing1Motion=Free
  swing2LimitDegrees=45
  swing2Motion=Free
  twistLimitDegrees=45
  twistMotion=Free
 }
 {
  bone1Name=dyn_collarBack_MSTR
  bone2Name=spine_05
  swing1LimitDegrees=45
  swing1Motion=Free
  swing2LimitDegrees=45
  swing2Motion=Free
  twistLimitDegrees=45
  twistMotion=Free
 }
 {
  bone1Name=neck_02
  bone2Name=spine_05
  swing1LimitDegrees=45
  swing1Motion=Free
  swing2LimitDegrees=45
  swing2Motion=Free
  twistLimitDegrees=45
  twistMotion=Free
 }
 {
  bone1Name=head
  bone2Name=neck_02
  swing1LimitDegrees=45
  swing1Motion=Free
  swing2LimitDegrees=45
  swing2Motion=Free
  twistLimitDegrees=45
  twistMotion=Free
 }
 {bone1Name=dyn_collarBack_r_01, bone2Name=dyn_collarBack_r_MSTR, swing1LimitDegrees=50, swing2LimitDegrees=20, twistLimitDegrees=20}
 {
  bone1Name=dyn_collarBack_r_MSTR
  bone2Name=spine_05
  swing1LimitDegrees=45
  swing1Motion=Free
  swing2LimitDegrees=45
  swing2Motion=Free
  twistLimitDegrees=45
  twistMotion=Free
 }
 {bone1Name=dyn_collarBack_l_01, bone2Name=dyn_collarBack_l_MSTR, swing1LimitDegrees=50, swing2LimitDegrees=10, twistLimitDegrees=20}
 {
  bone1Name=dyn_collarBack_l_MSTR
  bone2Name=spine_05
  swing1LimitDegrees=45
  swing1Motion=Free
  swing2LimitDegrees=45
  swing2Motion=Free
  twistLimitDegrees=45
  twistMotion=Free
 }
 {bone1Name=dyn_cloakFront_c_01, bone2Name=dyn_cloakFront_c_MSTR, swing1LimitDegrees=30, swing2LimitDegrees=5, twistLimitDegrees=0}
 {bone1Name=dyn_cloakFront_c_02, bone2Name=dyn_cloakFront_c_01, swing1LimitDegrees=30, swing2LimitDegrees=30, twistLimitDegrees=0}
 {bone1Name=dyn_cloakBack_c_01, bone2Name=dyn_cloakBack_c_MSTR, swing1LimitDegrees=45, swing2LimitDegrees=0, twistLimitDegrees=0}
 {bone1Name=dyn_cloakBack_c_02, bone2Name=dyn_cloakBack_c_01, swing1LimitDegrees=30, swing2LimitDegrees=15, twistLimitDegrees=0}
 {
  bone1Name=dyn_cloakBack_c_MSTR
  bone2Name=spine_05
  swing1LimitDegrees=45
  swing1Motion=Free
  swing2LimitDegrees=45
  swing2Motion=Free
  twistLimitDegrees=45
  twistMotion=Free
 }
 {
  bone1Name=dyn_cloakFront_c_MSTR
  bone2Name=spine_04
  swing1LimitDegrees=45
  swing1Motion=Free
  swing2LimitDegrees=45
  swing2Motion=Free
  twistLimitDegrees=45
  twistMotion=Free
 }
 {
  bone1Name=spine_03
  bone2Name=spine_02
  swing1LimitDegrees=45
  swing1Motion=Free
  swing2LimitDegrees=45
  swing2Motion=Free
  twistLimitDegrees=45
  twistMotion=Free
 }
 {bone1Name=dyn_ropeFront_A_MSTR, bone2Name=dyn_cloakFront_c_02, swing1LimitDegrees=20, swing2LimitDegrees=20, twistLimitDegrees=0}
 {bone1Name=dyn_ropeFront_B_MSTR, bone2Name=dyn_cloakFront_c_01, swing1LimitDegrees=20, swing2LimitDegrees=20, twistLimitDegrees=0}
 {bone1Name=dyn_ropeFront_C_MSTR, bone2Name=spine_04, swing1LimitDegrees=20, swing2LimitDegrees=20, twistLimitDegrees=0}
 {bone1Name=dyn_ropeBack_D_MSTR, bone2Name=dyn_cloakBack_c_02, swing1LimitDegrees=30, swing2LimitDegrees=30, twistLimitDegrees=0}
 {bone1Name=dyn_ropeBack_C_MSTR, bone2Name=dyn_cloakBack_c_01, swing1LimitDegrees=30, swing2LimitDegrees=30, twistLimitDegrees=0}
 {bone1Name=dyn_ropeBack_E_MSTR, bone2Name=dyn_cloakBack_c_02, swing1LimitDegrees=20, swing2LimitDegrees=20, twistLimitDegrees=0}
 {
  bone1Name=upperarm_l
  bone2Name=clavicle_l
  swing1LimitDegrees=45
  swing1Motion=Free
  swing2LimitDegrees=45
  swing2Motion=Free
  twistLimitDegrees=45
  twistMotion=Free
 }
 {
  bone1Name=upperarm_r
  bone2Name=clavicle_r
  swing1LimitDegrees=45
  swing1Motion=Free
  swing2LimitDegrees=45
  swing2Motion=Free
  twistLimitDegrees=45
  twistMotion=Free
 }
 {
  bone1Name=dyn_ropeFront_D_MSTR
  bone2Name=clavicle_l
  swing1LimitDegrees=45
  swing1Motion=Free
  swing2LimitDegrees=45
  swing2Motion=Free
  twistLimitDegrees=45
  twistMotion=Free
 }
 {
  bone1Name=dyn_ropeBack_B_MSTR
  bone2Name=clavicle_l
  swing1LimitDegrees=45
  swing1Motion=Free
  swing2LimitDegrees=45
  swing2Motion=Free
  twistLimitDegrees=45
  twistMotion=Free
 }
 {
  bone1Name=dyn_ropeBack_A_MSTR
  bone2Name=clavicle_l
  swing1LimitDegrees=45
  swing1Motion=Free
  swing2LimitDegrees=45
  swing2Motion=Free
  twistLimitDegrees=45
  twistMotion=Free
 }
 {bone1Name=dyn_ropeBack_A_01, bone2Name=dyn_ropeBack_A_MSTR, swing1LimitDegrees=45, swing2LimitDegrees=45, twistLimitDegrees=0}
 {bone1Name=dyn_ropeBack_B_01, bone2Name=dyn_ropeBack_B_MSTR, swing1LimitDegrees=30, swing2LimitDegrees=30, twistLimitDegrees=0}
 {bone1Name=dyn_ropeFront_D_01, bone2Name=dyn_ropeFront_D_MSTR, swing1LimitDegrees=30, swing2LimitDegrees=20, twistLimitDegrees=0}
 {bone1Name=dyn_ropeFront_C_01, bone2Name=dyn_ropeFront_C_MSTR, swing1LimitDegrees=50, swing2LimitDegrees=50, twistLimitDegrees=0}
]
