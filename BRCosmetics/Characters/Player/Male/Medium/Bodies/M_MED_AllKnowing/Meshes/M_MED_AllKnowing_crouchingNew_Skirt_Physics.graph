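# Physics asset for the crouching skirt variant of the male medium AllKnowing player body, defining collision and joints for skirt motion.
# Eight three-bone skirt chains of boxes and capsules hang from the pelvis or thighs, with per-chain swing limits and zero twist limits.
# Capsules on pelvis, thighs, calves and feet shape the lower body.
# The root adds extra shapes, including a huge sunken capsule whose top approximates a floor.
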PhysicsAsset /BRCosmetics/Characters/Player/Male/Medium/Bodies/M_MED_AllKnowing/Meshes/M_MED_AllKnowing_crouchingNew_Skirt_Physics
bodies {
 calf_l [{shapeType=Capsule, length=40.5119, radius=7.2541, center=<-19.7156 -1.9588 0>, rotation=<rot 90 0 0>}]
 calf_r [{shapeType=Capsule, length=43.0754, radius=9.6985, center=<19.7156 1.289 0>, rotation=<rot 90 0 180>}]
 dyn_skirtBack_l_01 [{shapeType=Box, extentX=17.972, extentY=17.8312, extentZ=2.37, shapeName=dyn_skirtBack_l_02_box, center=<4.0142 0 0>}]
 dyn_skirtBack_l_02 [{shapeType=Box, extentX=17.972, extentY=17.8312, extentZ=2.37, shapeName=dyn_skirtBack_l_02_box, center=<4.0142 0 0>}]
 dyn_skirtBack_l_03 [{
   shapeType=Box
   extentX=14.5503
   extentY=19.5567
   extentZ=2.6082
   shapeName=dyn_skirtBack_l_03_box
   center=<6.0275 -0.3 -1.7>
   rotation=<rot 0 0 10.1612>
  }]
 dyn_skirtBack_r_01 [{shapeType=Capsule, length=6.1864, radius=8.6129, shapeName=dyn_skirtBack_r_MSTR_capsule, center=<6.1105 0 10.4361>, rotation=<rot 90 180 90>}]
 dyn_skirtBack_r_02 [{shapeType=Capsule, length=6.1864, radius=8.6129, shapeName=dyn_skirtBack_r_02_capsule, center=<6.1105 0 10.4361>, rotation=<rot 90 180 90>}]
 dyn_skirtBack_r_03 [{shapeType=Capsule, length=5.6011, radius=9.1982, shapeName=dyn_skirtBack_r_03_capsule, center=<6.1105 0 10.4361>, rotation=<rot 90 180 90>}]
 dyn_skirtFrontA_r_01 [{
   shapeType=Capsule
   length=17.7857
   radius=8.2198
   shapeName=dyn_skirtFrontA_r_02_capsule
   center=<-1.9796 0 -7.7>
   rotation=<rot 89.1017 -90.0072 179.9927>
  }]
 dyn_skirtFrontA_r_02 [{
   shapeType=Capsule
   length=17.7857
   radius=8.2198
   shapeName=dyn_skirtFrontA_r_02_capsule
   center=<-1.9796 0 -7.7>
   rotation=<rot 89.1017 -90.0072 179.9927>
  }]
 dyn_skirtFrontA_r_03 [{
   shapeType=Capsule
   length=17.6292
   radius=8.3764
   shapeName=dyn_skirtFrontA_r_03_capsule
   center=<-2.8372 0 -7.7>
   rotation=<rot 89.1017 -90.0071 179.9928>
  }]
 dyn_skirtFront_l_01 [{shapeType=Capsule, length=1, radius=1, shapeName=dyn_skirtFront_r_03_capsule, center=<3.7203 0 8.9>, rotation=<rot 0 0 90>}]
 dyn_skirtFront_l_02 [{shapeType=Box, extentX=15.2138, extentY=21.0313, extentZ=3.2371, shapeName=dyn_skirtFront_l_02_box, center=<4.3111 0 0>}]
 dyn_skirtFront_l_03 [{shapeType=Box, extentX=13.4297, extentY=21.0313, extentZ=3.2371, shapeName=dyn_skirtFront_l_03_box, center=<4.9635 0 0>}]
 dyn_skirtFront_r_01 [{shapeType=Capsule, length=15.1893, radius=7.2432, shapeName=dyn_skirtFront_r_03_capsule, center=<3.7203 0 8.9>, rotation=<rot 90 90 0>}]
 dyn_skirtFront_r_02 [{shapeType=Capsule, length=11.8371, radius=7.2432, shapeName=dyn_skirtFront_r_03_capsule, center=<3.7203 0 8.9>, rotation=<rot 90 90 0>}]
 dyn_skirtFront_r_03 [{shapeType=Capsule, length=15.1893, radius=7.2432, shapeName=dyn_skirtFront_r_03_capsule, center=<3.7203 0 8.9>, rotation=<rot 90 90 0>}]
 dyn_skirtSideA_l_01 [{shapeType=Capsule, length=18.3678, radius=8.0577, shapeName=dyn_skirtSideA_l_01_capsule, center=<3.6398 0 6.4>, rotation=<rot 90 90 0>}]
 dyn_skirtSideA_l_02 [{shapeType=Capsule, length=18.3678, radius=8.0577, shapeName=dyn_skirtSideA_l_02_capsule, center=<3.6398 0 6.4>, rotation=<rot 90 90 0>}]
 dyn_skirtSideA_l_03 [{
   shapeType=Capsule
   length=21.1172
   radius=9.9425
   shapeName=dyn_skirtSideA_l_03_capsule
   center=<7.3286 0 6.4>
   rotation=<rot 90 165.9638 75.9638>
  }]
 dyn_skirtSideA_r_01 [{shapeType=Capsule, length=18.0602, radius=7.4993, shapeName=dyn_skirtSideA_r_02_capsule, center=<-3.8594 0 -6.7>, rotation=<rot 90 90 0>}]
 dyn_skirtSideA_r_02 [{shapeType=Capsule, length=18.0602, radius=7.4993, shapeName=dyn_skirtSideA_r_02_capsule, center=<-3.8594 0 -6.7>, rotation=<rot 90 90 0>}]
 dyn_skirtSideA_r_03 [{
   shapeType=Capsule
   length=31.9608
   radius=9.7892
   shapeName=dyn_skirtSideA_r_03_capsule
   center=<-6.3926 0 -6.7>
   rotation=<rot 90 165.9638 75.9638>
  }]
 dyn_skirtSide_r_01 [{shapeType=Capsule, length=1, radius=1, shapeName=dyn_skirtSideA_r_02_capsule, center=<-3.8594 0 -6.7>, rotation=<rot 90 90 0>}]
 dyn_skirtSide_r_02 [{shapeType=Capsule, length=18.0602, radius=7.4993, shapeName=dyn_skirtSideA_r_02_capsule, center=<-3.8594 0 -6.7>, rotation=<rot 90 90 0>}]
 dyn_skirtSide_r_03 [{shapeType=Capsule, length=18.0602, radius=7.4993, shapeName=dyn_skirtSideA_r_02_capsule, center=<-3.8594 0 -6.7>, rotation=<rot 90 90 0>}]
 foot_l [{
   shapeType=Capsule
   length=12.5015
   radius=7.8269
   shapeName=foot_l_capsule
   center=<-5.0619 7.6678 0.0543>
   rotation=<rot 0.1321 -1.0968 -89.5952>
  }]
 foot_r [{
   shapeType=Capsule
   length=12.5015
   radius=7.8269
   shapeName=foot_l_capsule
   center=<5.0619 -7.6678 -0.0543>
   rotation=<rot 0.1321 -1.0968 90.4048>
  }]
 pelvis [{shapeType=Capsule, length=9.159, radius=12.7688, center=<0 -1.3004 0>}]
 root [
  {shapeType=Sphere, radius=16.9907, shapeName=root_sphere, center=<4.2844 0 15.5315>}
  {shapeType=Capsule, length=0.505, radius=300, shapeName=root_capsule, center=<0 0 -293.2007>, rotation=<rot 0 -90 0>}
  {
   shapeType=Capsule
   length=17.2416
   radius=17.5411
   shapeName=root_capsule
   center=<-7.6233 -26.575 6.5955>
   rotation=<rot -87.9422 0.0036 -98.6601>
  }
 ]
 thigh_l [{shapeType=Capsule, length=33.0373, radius=7.8918, center=<-19.686 -0.6941 0>, rotation=<rot 90 180 177.5578>}]
 thigh_r [{shapeType=Capsule, length=34.549, radius=9.6329, center=<21.8558 0 0>, rotation=<rot 90 0 180>}]
}
constraints [
 {bone1Name=thigh_l, bone2Name=pelvis, swing1LimitDegrees=45, swing2LimitDegrees=45, twistLimitDegrees=45}
 {bone1Name=calf_l, bone2Name=thigh_l, swing1LimitDegrees=45, swing2LimitDegrees=45, twistLimitDegrees=45}
 {bone1Name=thigh_r, bone2Name=pelvis, swing1LimitDegrees=45, swing2LimitDegrees=45, twistLimitDegrees=45}
 {bone1Name=calf_r, bone2Name=thigh_r, swing1LimitDegrees=45, swing2LimitDegrees=45, twistLimitDegrees=45}
 {
  bone1Name=foot_l
  bone2Name=calf_l
  swing1LimitDegrees=45
  swing1Motion=Free
  swing2LimitDegrees=45
  swing2Motion=Free
  twistLimitDegrees=45
  twistMotion=Free
 }
 {
  bone1Name=foot_r
  bone2Name=calf_r
  swing1LimitDegrees=45
  swing1Motion=Free
  swing2LimitDegrees=45
  swing2Motion=Free
  twistLimitDegrees=45
  twistMotion=Free
 }
 {bone1Name=dyn_skirtBack_l_02, bone2Name=dyn_skirtBack_l_01, swing1LimitDegrees=20, swing2LimitDegrees=45, twistLimitDegrees=0}
 {bone1Name=dyn_skirtBack_l_03, bone2Name=dyn_skirtBack_l_02, swing1LimitDegrees=20, swing2LimitDegrees=45, twistLimitDegrees=0}
 {bone1Name=dyn_skirtFront_l_01, bone2Name=thigh_l, swing1LimitDegrees=20, swing2LimitDegrees=40, twistLimitDegrees=0}
 {bone1Name=dyn_skirtFront_l_02, bone2Name=dyn_skirtFront_l_01, swing1LimitDegrees=20, swing2LimitDegrees=40, twistLimitDegrees=0}
 {bone1Name=dyn_skirtFront_l_03, bone2Name=dyn_skirtFront_l_02, swing1LimitDegrees=20, swing2LimitDegrees=40, twistLimitDegrees=0}
 {bone1Name=dyn_skirtSideA_l_01, bone2Name=thigh_l, swing1LimitDegrees=20, swing2LimitDegrees=20, twistLimitDegrees=0}
 {bone1Name=dyn_skirtSideA_l_02, bone2Name=dyn_skirtSideA_l_01, swing1LimitDegrees=20, swing2LimitDegrees=20, twistLimitDegrees=0}
 {bone1Name=dyn_skirtSideA_l_03, bone2Name=dyn_skirtSideA_l_02, swing1LimitDegrees=20, swing2LimitDegrees=20, twistLimitDegrees=0}
 {bone1Name=dyn_skirtBack_r_02, bone2Name=dyn_skirtBack_r_01, swing1LimitDegrees=30, swing2LimitDegrees=55, twistLimitDegrees=0}
 {bone1Name=dyn_skirtBack_r_03, bone2Name=dyn_skirtBack_r_02, swing1LimitDegrees=30, swing2LimitDegrees=45, twistLimitDegrees=0}
 {bone1Name=dyn_skirtFrontA_r_01, bone2Name=thigh_r, swing1LimitDegrees=20, swing2LimitDegrees=35, twistLimitDegrees=0}
 {bone1Name=dyn_skirtFrontA_r_02, bone2Name=dyn_skirtFrontA_r_01, swing1LimitDegrees=20, swing2LimitDegrees=35, twistLimitDegrees=0}
 {bone1Name=dyn_skirtFrontA_r_03, bone2Name=dyn_skirtFrontA_r_02, swing1LimitDegrees=20, swing2LimitDegrees=35, twistLimitDegrees=0}
 {bone1Name=dyn_skirtSideA_r_02, bone2Name=dyn_skirtSideA_r_01, swing1LimitDegrees=10, swing2LimitDegrees=45, twistLimitDegrees=0}
 {bone1Name=dyn_skirtSideA_r_03, bone2Name=dyn_skirtSideA_r_02, swing1LimitDegrees=10, swing2LimitDegrees=45, twistLimitDegrees=0}
 {bone1Name=dyn_skirtBack_l_01, bone2Name=pelvis, swing1LimitDegrees=20, swing2LimitDegrees=45, twistLimitDegrees=0}
 {bone1Name=dyn_skirtBack_r_01, bone2Name=pelvis, swing1LimitDegrees=30, swing2LimitDegrees=45, twistLimitDegrees=0}
 {bone1Name=dyn_skirtSideA_r_01, bone2Name=pelvis, swing1LimitDegrees=10, swing2LimitDegrees=45, twistLimitDegrees=0}
 {bone1Name=dyn_skirtFront_r_01, bone2Name=pelvis, swing1LimitDegrees=30, swing2LimitDegrees=30, twistLimitDegrees=0}
 {bone1Name=dyn_skirtFront_r_02, bone2Name=dyn_skirtFront_r_01, swing1LimitDegrees=15, swing2LimitDegrees=30, twistLimitDegrees=0}
 {bone1Name=dyn_skirtFront_r_03, bone2Name=dyn_skirtFront_r_02, swing1LimitDegrees=15, swing2LimitDegrees=30, twistLimitDegrees=0}
 {bone1Name=dyn_skirtSide_r_03, bone2Name=dyn_skirtSide_r_02, swing1LimitDegrees=20, swing2LimitDegrees=30, twistLimitDegrees=0}
 {bone1Name=dyn_skirtSide_r_02, bone2Name=dyn_skirtSide_r_01, swing1LimitDegrees=20, swing2LimitDegrees=30, twistLimitDegrees=0}
 {bone1Name=dyn_skirtSide_r_01, bone2Name=thigh_r, swing1LimitDegrees=20, swing2LimitDegrees=30, twistLimitDegrees=0}
 {
  bone1Name=pelvis
  bone2Name=root
  swing1LimitDegrees=45
  swing1Motion=Free
  swing2LimitDegrees=45
  swing2Motion=Free
  twistLimitDegrees=45
  twistMotion=Free
 }
]
