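# Animation Blueprint for the AllKnowing medium male character body, copying the parent mesh pose and simulating its cloak, ropes and skirt.
# Control Rig and weighted constraints feed RigidBody cloak simulation, switched for back bling.
# Skirt physics then switches between default, front-end and crouch setups, followed by additive bone fixes for thigh lift, crouch, front end and skydiving.
# The update event computes gravity override, back bling, crouch flags and thigh-rotation alpha.
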AnimBlueprint /BRCosmetics/Characters/Player/Male/Medium/Bodies/M_MED_AllKnowing/Meshes/M_MED_AllKnowing_AnimBP
target_skeleton=&/M_MED_AllKnowing_Skeleton
graphs [
 AnimGraph {nodes [
   Constraint!(ComponentPose, Alpha=1) -> (Pose) {alphaBoolBlend {blendOption=Linear}, boneToModify=~}
   "Constraint - dyn_cloakBack_c_01 (PT)"!(ComponentPose, Alpha=1, ConstraintWeights_0=1, ConstraintWeights_1:float=1) -> (Pose) {
    alphaBoolBlend {blendOption=Linear}
    boneToModify {boneName=dyn_cloakBack_c_01}
    constraintSetup [{targetBone {boneName=spine_04}}, {transformType=Translation, targetBone {boneName=spine_05}}]
    constraintWeights [1,1]
   }
   n1 = "Transform (Modify) Bone - Bone: dyn_skirtFrontA_r_01"(
    ComponentPose=Local To Component(LocalPose=Use cached pose 'skirtPose'())
    Alpha=Get thig_r_RotUp() -> (thig_r_RotUp:float)
    Translation="0,0,0"
    Rotation="10,0,0"
    Scale="1,1,1"
   ) {
    rotationMode=BMM_Additive
    rotationSpace=BCS_BoneSpace
    translationMode=BMM_Additive
    translationSpace=BCS_BoneSpace
    alphaBoolBlend {blendOption=Linear}
    boneToModify {boneName=dyn_skirtFrontA_r_01}
    rotation {pitch=10}
   }
   n2 = "Transform (Modify) Bone - Bone: dyn_skirtFront_r_01"!(ComponentPose=@n1, bAlphaBoolEnabled:bool=false, Translation="-0.765,7.766,8.167", Rotation="25.780,-0.239,40.781", Scale="1,1,1") -> (Pose) {
    alphaInputType=Bool
    rotationMode=BMM_Additive
    rotationSpace=BCS_BoneSpace
    translationMode=BMM_Additive
    translationSpace=BCS_BoneSpace
    alphaBoolBlend {blendOption=Linear}
    boneToModify {boneName=dyn_skirtFront_r_01}
    rotation=<rot 25.78 -0.239 40.781>
    translation=<-0.765 7.766 8.167>
   }
   n3 = Get bIsCrouching()
   n4 = "Transform (Modify) Bone - Bone: dyn_skirtBack_r_MSTR"!(ComponentPose=@n2, bAlphaBoolEnabled:bool=@n3, Translation="0,0,0", Rotation="0,0,0", Scale="1,1,1") -> (Pose) {
    alphaInputType=Bool
    rotationMode=BMM_Additive
    rotationSpace=BCS_BoneSpace
    translationMode=BMM_Additive
    translationSpace=BCS_BoneSpace
    alphaBoolBlend {blendOption=Linear}
    boneToModify {boneName=dyn_skirtBack_r_MSTR}
   }
   n5 = "Transform (Modify) Bone - Bone: dyn_skirtSideA_r_01"!(ComponentPose=@n4, bAlphaBoolEnabled:bool=@n3, Translation="0,0,0", Rotation="15,0,0", Scale="1,1,1") -> (Pose) {
    alphaInputType=Bool
    rotationMode=BMM_Additive
    rotationSpace=BCS_BoneSpace
    translationMode=BMM_Additive
    translationSpace=BCS_BoneSpace
    alphaBoolBlend {blendOption=Linear}
    boneToModify {boneName=dyn_skirtSideA_r_01}
    rotation {pitch=15}
   }
   n6 = "Transform (Modify) Bone - Bone: dyn_skirtFront_l_03"!(ComponentPose=@n5, bAlphaBoolEnabled:bool=@n3, Translation="0,0,0", Rotation="-10,0,0", Scale="1,1,1") -> (Pose) {
    alphaInputType=Bool
    rotationMode=BMM_Additive
    rotationSpace=BCS_BoneSpace
    translationMode=BMM_Additive
    translationSpace=BCS_BoneSpace
    alphaBoolBlend {blendOption=Linear}
    boneToModify {boneName=dyn_skirtFront_l_03}
    rotation {pitch=-10}
   }
   n7 = "Transform (Modify) Bone - Bone: dyn_skirtFront_l_03"(ComponentPose=@n6, bAlphaBoolEnabled=@n3, Translation="0,0,0", Rotation="-10,0,0", Scale="1,1,1") {
    alphaInputType=Bool
    rotationMode=BMM_Additive
    rotationSpace=BCS_BoneSpace
    translationMode=BMM_Additive
    translationSpace=BCS_BoneSpace
    alphaBoolBlend {blendOption=Linear}
    boneToModify {boneName=dyn_skirtFront_l_03}
    rotation {pitch=-10}
   }
   n8 = Get bIsInFrontEnd()
   n9 = "Transform (Modify) Bone - Bone: dyn_skirtFront_r_01"(ComponentPose=@n7, bAlphaBoolEnabled=@n8, Translation="0,1,0", Rotation="-10,10,0", Scale="0,1,1") {
    alphaInputType=Bool
    rotationMode=BMM_Additive
    rotationSpace=BCS_BoneSpace
    translationMode=BMM_Additive
    translationSpace=BCS_BoneSpace
    alphaBoolBlend {blendOption=Linear}
    boneToModify {boneName=dyn_skirtFront_r_01}
    rotation {pitch=-10, yaw=10}
    scale {x=0}
    translation {y=1}
   }
   n10 = "Transform (Modify) Bone - Bone: dyn_skirtFront_r_02"!(ComponentPose=@n9, bAlphaBoolEnabled:bool=@n8, Translation="-0.341,-1.995,-0.700", Rotation="0.870,-4.888,1.408", Scale="0,1,1") -> (Pose) {
    alphaInputType=Bool
    rotationMode=BMM_Additive
    rotationSpace=BCS_BoneSpace
    translationMode=BMM_Additive
    translationSpace=BCS_BoneSpace
    alphaBoolBlend {blendOption=Linear}
    boneToModify {boneName=dyn_skirtFront_r_02}
    rotation=<rot 0.8701 -4.888 1.408>
    scale {x=0}
    translation=<-0.341 -1.995 -0.7>
   }
   n11 = "Transform (Modify) Bone - Bone: dyn_skirtFront_r_03"!(ComponentPose=@n10, bAlphaBoolEnabled:bool=@n8, Translation="0,0,0", Rotation="0,-10,0", Scale="0,1,1") -> (Pose) {
    alphaInputType=Bool
    rotationMode=BMM_Additive
    rotationSpace=BCS_BoneSpace
    translationMode=BMM_Additive
    translationSpace=BCS_BoneSpace
    alphaBoolBlend {blendOption=Linear}
    boneToModify {boneName=dyn_skirtFront_r_03}
    rotation {yaw=-10}
    scale {x=0}
   }
   n12 = "Transform (Modify) Bone - Bone: dyn_skirtFront_l_01"!(ComponentPose=@n11, bAlphaBoolEnabled:bool=@n8, Translation="0,0,-2.5", Rotation="-2.5,0,0", Scale="0,1,1") -> (Pose) {
    alphaInputType=Bool
    rotationMode=BMM_Additive
    rotationSpace=BCS_BoneSpace
    translationMode=BMM_Additive
    translationSpace=BCS_BoneSpace
    alphaBoolBlend {blendOption=Linear}
    boneToModify {boneName=dyn_skirtFront_l_01}
    rotation {pitch=-2.5}
    scale {x=0}
    translation {z=-2.5}
   }
   n13 = "Transform (Modify) Bone - Bone: neck_02"!(ComponentPose=@n12, bAlphaBoolEnabled:bool=Get bIsSkydiveFloating(), Translation="0,-12,0", Rotation="0,0,0", Scale="1,1,1") -> (Pose) {
    alphaInputType=Bool
    rotationMode=BMM_Additive
    rotationSpace=BCS_BoneSpace
    translationMode=BMM_Additive
    translationSpace=BCS_BoneSpace
    alphaBoolBlend {blendOption=Linear}
    boneToModify {boneName=neck_02}
    translation {y=-12}
   }
   n14 = "Transform (Modify) Bone - Bone: neck_01"!(
    ComponentPose=@n13
    bAlphaBoolEnabled:bool=Get isRigidBodyDisabled() -> (isRigidBodyDisabled:bool)
    Translation="-5,7,0"
    Rotation="-0.000,30,0.000"
    Scale="1,1,1"
   ) -> (Pose) {
    alphaInputType=Bool
    rotationMode=BMM_Additive
    rotationSpace=BCS_BoneSpace
    translationMode=BMM_Additive
    translationSpace=BCS_BoneSpace
    alphaBoolBlend {blendOption=Linear}
    boneToModify {boneName=neck_01}
    rotation {yaw=30}
    translation=<-5 7>
   }
   Output Pose(Result=Component To Local(ComponentPose=@n14)) {layerGroup=None}
   n15 = Control Rig(Source=Copy Pose From Mesh() {bUseAttachedParent=true}) {
    alphaBoolBlend {blendOption=Linear}
    controlRigAssetReference {blueprintRigClass=&/M_MED_DeformRig_AllKnowing_ControlRig.M_MED_DeformRig_AllKnowing_ControlRig_C}
   }
   n16 = "Constraint - dyn_ropeFront_D_MSTR (PPPPPP)"!(
    ComponentPose=Local To Component(LocalPose=@n15)
    Alpha=1
    ConstraintWeights_0=0.165
    ConstraintWeights_1:float=0.224
    ConstraintWeights_2:float=0.239
    ConstraintWeights_3:float=0.192
    ConstraintWeights_4:float=0.082
    ConstraintWeights_5:float=0.097
   ) -> (Pose) {
    alphaBoolBlend {blendOption=Linear}
    boneToModify {boneName=dyn_ropeFront_D_MSTR}
    constraintSetup [
     {targetBone {boneName=deform_upperarm_fr_l}}
     {targetBone {boneName=deform_upperarm_up_l}}
     {targetBone {boneName=upperarm_twist_01_l}}
     {targetBone {boneName=clavicle_l}}
     {targetBone {boneName=spine_05}}
     {targetBone {boneName=upperarm_twist_02_l}}
    ]
    constraintWeights [1,1,1,1,1,1]
   }
   n17 = "Constraint - dyn_ropeFront_C_MSTR (PPPPPPP)"!(
    ComponentPose=@n16
    Alpha=1
    ConstraintWeights_0=0.189
    ConstraintWeights_1:float=0.288
    ConstraintWeights_2:float=0.142
    ConstraintWeights_3:float=0.096
    ConstraintWeights_4:float=0.098
    ConstraintWeights_5:float=0.091
    ConstraintWeights_6:float=0.1
   ) -> (Pose) {
    alphaBoolBlend {blendOption=Linear}
    boneToModify {boneName=dyn_ropeFront_C_MSTR}
    constraintSetup [
     {targetBone {boneName=clavicle_l}}
     {targetBone {boneName=deform_pec_l}}
     {targetBone {boneName=spine_04}}
     {targetBone {boneName=deform_upperarm_fr_l}}
     {targetBone {boneName=spine_03}}
     {targetBone {boneName=spine_05}}
     {targetBone {boneName=upperarm_twist_01_l}}
    ]
    constraintWeights [1,1,1,1,1,1,1]
   }
   n18 = "Constraint - dyn_ropeBack_A_MSTR (PPPPPPP)"!(
    ComponentPose=@n17
    Alpha=1
    ConstraintWeights_0=0.207
    ConstraintWeights_1:float=0.106
    ConstraintWeights_2:float=0.27
    ConstraintWeights_3:float=0.294
    ConstraintWeights_4:float=0.04
    ConstraintWeights_5:float=0.05
    ConstraintWeights_6:float=0.03
   ) -> (Pose) {
    alphaBoolBlend {blendOption=Linear}
    boneToModify {boneName=dyn_ropeBack_A_MSTR}
    constraintSetup [
     {targetBone {boneName=clavicle_l}}
     {targetBone {boneName=deform_upperarm_bk_l}}
     {targetBone {boneName=deform_upperarm_up_l}}
     {targetBone {boneName=upperarm_twist_01_l}}
     {targetBone {boneName=spine_03}}
     {targetBone {boneName=upperarm_twist_02_l}}
     {targetBone {boneName=spine_05}}
    ]
    constraintWeights [1,1,1,1,1,1,1]
   }
   n19 = "Constraint - dyn_ropeBack_B_MSTR (PPPPP)"!(
    ComponentPose=@n18
    Alpha=1
    ConstraintWeights_0=0.3
    ConstraintWeights_1:float=0.2
    ConstraintWeights_2:float=0.1
    ConstraintWeights_3:float=0.2
    ConstraintWeights_4:float=0.2
   ) -> (Pose) {
    alphaBoolBlend {blendOption=Linear}
    boneToModify {boneName=dyn_ropeBack_B_MSTR}
    constraintSetup [
     {targetBone {boneName=clavicle_l}}
     {targetBone {boneName=deform_upperarm_bk_l}}
     {targetBone {boneName=dyn_cloakBack_c_01}}
     {targetBone {boneName=spine_03}}
     {targetBone {boneName=upperarm_twist_01_l}}
    ]
    constraintWeights [1,1,1,1,1]
   }
   Save cached pose 'defaultPose'(Pose=Component To Local(ComponentPose=@n19)) {cacheName=defaultPose}
   n20 = RigidBody!(
    ComponentPose=Local To Component(LocalPose=Use cached pose 'defaultPose'())
    Alpha
    OverrideWorldGravity:Vector=Get Gravity Override()
    ExternalForce
   ) -> (Pose) {
    bClampLinearTranslationLimitToRefPose=true
    bOverrideWorldGravity=true
    overridePhysicsAsset=&/M_MED_AllKnowing_cloak_Physics
    simulationSpace=BaseBoneSpace
    alphaBoolBlend {blendOption=Linear}
    baseBoneRef {boneName=spine_04}
    componentAppliedLinearAccClamp=<6000 6000 6000>
    componentLinearAccScale=<0.6 0.2 0.2>
    componentLinearVelScale=<0.6 0.2 0.2>
   }
   n21 = RigidBody(ComponentPose=Local To Component(LocalPose=Use cached pose 'defaultPose'()), OverrideWorldGravity=Get Gravity Override()) {
    bOverrideWorldGravity=true
    overridePhysicsAsset=&/M_MED_AllKnowing_BackBlingNew_Physics
    simulationSpace=BaseBoneSpace
    alphaBoolBlend {blendOption=Linear}
    baseBoneRef {boneName=spine_05}
    componentAppliedLinearAccClamp=<6000 6000 6000>
    componentLinearAccScale=<0.6 0.2 0.2>
    componentLinearVelScale=<0.6 0.2 0.2>
   }
   n22 = "Transform (Modify) Bone - Bone: dyn_cloakBack_c_01"(ComponentPose=@n21, Alpha=1, Translation="0,0,0", Rotation="0,7,0", Scale="1,1,1") {
    rotationMode=BMM_Additive
    rotationSpace=BCS_BoneSpace
    alphaBoolBlend {blendOption=Linear}
    boneToModify {boneName=dyn_cloakBack_c_01}
    rotation {yaw=7}
   }
   Save cached pose 'upperBodyCache'(Pose=Blend Poses by bool(
     BlendPose_0=Component To Local(ComponentPose=@n22)
     BlendPose_1=Component To Local(ComponentPose=@n20)
     bActiveValue=Get Is Wearing Backbling()
    ) {blendType=Linear}) {cacheName=upperBodyCache}
   n23 = "Transform (Modify) Bone - Bone: dyn_skirtFront_l_01"!(
    ComponentPose=Local To Component(LocalPose=Use cached pose 'upperBodyCache'())
    Alpha:float=1
    Translation="0,0,0"
    Rotation="0,0,0"
    Scale="1,1,1"
   ) -> (Pose) {
    rotationMode=BMM_Additive
    rotationSpace=BCS_BoneSpace
    scaleMode=BMM_Additive
    scaleSpace=BCS_BoneSpace
    translationMode=BMM_Additive
    translationSpace=BCS_BoneSpace
    alphaBoolBlend {blendOption=Linear}
    boneToModify {boneName=dyn_skirtFront_l_01}
   }
   n24 = "Transform (Modify) Bone - Bone: dyn_skirtFront_l_02"(ComponentPose=@n23, Alpha=1, Translation="0,0,0", Rotation="0,0,0", Scale="1,1,1") {
    rotationMode=BMM_Additive
    rotationSpace=BCS_BoneSpace
    scaleMode=BMM_Additive
    scaleSpace=BCS_BoneSpace
    translationMode=BMM_Additive
    translationSpace=BCS_BoneSpace
    alphaBoolBlend {blendOption=Linear}
    boneToModify {boneName=dyn_skirtFront_l_02}
   }
   n25 = "Transform (Modify) Bone - Bone: dyn_skirtFront_l_03"!(ComponentPose=@n24, Alpha:float=1, Translation="0,0,0", Rotation="0,0,0", Scale="1,1,1") -> (Pose) {
    rotationMode=BMM_Additive
    rotationSpace=BCS_BoneSpace
    scaleMode=BMM_Additive
    scaleSpace=BCS_BoneSpace
    translationMode=BMM_Additive
    translationSpace=BCS_BoneSpace
    alphaBoolBlend {blendOption=Linear}
    boneToModify {boneName=dyn_skirtFront_l_03}
   }
   n26 = "Transform (Modify) Bone - Bone: dyn_skirtSideA_l_01"(ComponentPose=@n25, Alpha=1, Translation="0,0,0", Rotation="0.009,-22.431,0.027", Scale="1,1,1") {
    rotationMode=BMM_Additive
    rotationSpace=BCS_BoneSpace
    scaleMode=BMM_Additive
    scaleSpace=BCS_BoneSpace
    translationMode=BMM_Additive
    translationSpace=BCS_BoneSpace
    alphaBoolBlend {blendOption=Linear}
    boneToModify {boneName=dyn_skirtSideA_l_01}
    rotation=<rot 0.009 -22.431 0.027>
   }
   n27 = "Transform (Modify) Bone - Bone: dyn_skirtSideA_l_02"(ComponentPose=@n26, Alpha=1, Translation="0,0,0", Rotation="0.004,-9.306,-0.021", Scale="1,1,1") {
    rotationMode=BMM_Additive
    rotationSpace=BCS_BoneSpace
    scaleMode=BMM_Additive
    scaleSpace=BCS_BoneSpace
    translationMode=BMM_Additive
    translationSpace=BCS_BoneSpace
    alphaBoolBlend {blendOption=Linear}
    boneToModify {boneName=dyn_skirtSideA_l_02}
    rotation=<rot 0.004 -9.306 -0.021>
   }
   n28 = "Transform (Modify) Bone - Bone: dyn_skirtBack_l_MSTR"(ComponentPose=@n27, Alpha=1, Translation="0,0,0", Rotation="-0.000,28.773,0.040", Scale="1,1,1") {
    rotationMode=BMM_Additive
    rotationSpace=BCS_BoneSpace
    scaleMode=BMM_Additive
    scaleSpace=BCS_BoneSpace
    translationMode=BMM_Additive
    translationSpace=BCS_BoneSpace
    alphaBoolBlend {blendOption=Linear}
    boneToModify {boneName=dyn_skirtBack_l_MSTR}
    rotation=<rot 0 28.773 0.04>
   }
   n29 = "Transform (Modify) Bone - Bone: dyn_skirtFront_r_01"!(ComponentPose=@n28, Alpha:float=1, Translation="-5.192,8.452,0.006", Rotation="0,0,0", Scale="1,1,1") -> (Pose) {
    rotationMode=BMM_Additive
    rotationSpace=BCS_BoneSpace
    scaleMode=BMM_Additive
    scaleSpace=BCS_BoneSpace
    translationMode=BMM_Additive
    translationSpace=BCS_BoneSpace
    alphaBoolBlend {blendOption=Linear}
    boneToModify {boneName=dyn_skirtFront_r_01}
    translation=<-5.192 8.452 0.006>
   }
   n30 = "Transform (Modify) Bone - Bone: dyn_skirtFrontA_r_01"(ComponentPose=@n29, Alpha=1, Translation="0.598,0.888,-1.827", Rotation="1.136,-5.720,-26.387", Scale="1,1,1") {
    rotationMode=BMM_Additive
    rotationSpace=BCS_BoneSpace
    scaleMode=BMM_Additive
    scaleSpace=BCS_BoneSpace
    translationMode=BMM_Additive
    translationSpace=BCS_BoneSpace
    alphaBoolBlend {blendOption=Linear}
    boneToModify {boneName=dyn_skirtFrontA_r_01}
    rotation=<rot 1.136 -5.72 -26.387>
    translation=<0.598 0.888 -1.827>
   }
   n31 = "Transform (Modify) Bone - Bone: dyn_skirtSide_r_01"(ComponentPose=@n30, Alpha=1, Translation="0.731,1.551,-0.200", Rotation="13.531,-26.559,0.011", Scale="1,1,1") {
    rotationMode=BMM_Additive
    rotationSpace=BCS_BoneSpace
    scaleMode=BMM_Additive
    scaleSpace=BCS_BoneSpace
    translationMode=BMM_Additive
    translationSpace=BCS_BoneSpace
    alphaBoolBlend {blendOption=Linear}
    boneToModify {boneName=dyn_skirtSide_r_01}
    rotation=<rot 13.531 -26.559 0.011>
    translation=<0.731 1.551 -0.2>
   }
   n32 = "Transform (Modify) Bone - Bone: dyn_skirtSideA_r_01"!(ComponentPose=@n31, Alpha:float=1, Translation="0,0,0", Rotation="49.768,0.039,-0.014", Scale="1,1,1") -> (Pose) {
    rotationMode=BMM_Additive
    rotationSpace=BCS_BoneSpace
    scaleMode=BMM_Additive
    scaleSpace=BCS_BoneSpace
    translationMode=BMM_Additive
    translationSpace=BCS_BoneSpace
    alphaBoolBlend {blendOption=Linear}
    boneToModify {boneName=dyn_skirtSideA_r_01}
    rotation=<rot 49.768 0.039 -0.014>
   }
   n33 = "Transform (Modify) Bone - Bone: dyn_skirtSideA_r_02"(ComponentPose=@n32, Alpha=1, Translation="0,0,0", Rotation="0,0,0", Scale="1,1,1") {
    rotationMode=BMM_Additive
    rotationSpace=BCS_BoneSpace
    scaleMode=BMM_Additive
    scaleSpace=BCS_BoneSpace
    translationMode=BMM_Additive
    translationSpace=BCS_BoneSpace
    alphaBoolBlend {blendOption=Linear}
    boneToModify {boneName=dyn_skirtSideA_r_02}
   }
   n34 = "Transform (Modify) Bone - Bone: dyn_skirtBack_r_MSTR"!(ComponentPose=@n33, Alpha:float=1, Translation="0,0,0", Rotation="38.284,-9.140,-5.700", Scale="1,1,1") -> (Pose) {
    rotationMode=BMM_Additive
    rotationSpace=BCS_BoneSpace
    scaleMode=BMM_Additive
    scaleSpace=BCS_BoneSpace
    translationMode=BMM_Additive
    translationSpace=BCS_BoneSpace
    alphaBoolBlend {blendOption=Linear}
    boneToModify {boneName=dyn_skirtBack_r_MSTR}
    rotation=<rot 38.284 -9.14 -5.7>
   }
   n35 = "Transform (Modify) Bone - Bone: dyn_skirtBack_r_02"(ComponentPose=@n34, Alpha=1, Translation="0,0,0", Rotation="0,0,0", Scale="1,1,1") {
    rotationMode=BMM_Additive
    rotationSpace=BCS_BoneSpace
    scaleMode=BMM_Additive
    scaleSpace=BCS_BoneSpace
    translationMode=BMM_Additive
    translationSpace=BCS_BoneSpace
    alphaBoolBlend {blendOption=Linear}
    boneToModify {boneName=dyn_skirtBack_r_02}
   }
   n36 = "Transform (Modify) Bone - Bone: deform_groin_base_l"!(
    ComponentPose=@n35
    bAlphaBoolEnabled:bool=Get crouch moving() -> (crouch moving:bool)
    Translation="-7.899,-4.023,11.973"
    Rotation="0,0,0"
    Scale="1,1,1"
   ) -> (Pose) {
    alphaInputType=Bool
    rotationMode=BMM_Additive
    rotationSpace=BCS_BoneSpace
    scaleMode=BMM_Additive
    scaleSpace=BCS_BoneSpace
    translationMode=BMM_Additive
    translationSpace=BCS_BoneSpace
    alphaBoolBlend {blendOption=Linear}
    boneToModify {boneName=deform_groin_base_l}
    translation=<-7.899 -4.023 11.973>
   }
   Save cached pose 'crouchPose'(Pose=Component To Local(ComponentPose=@n36)) {cacheName=crouchPose}
   n37 = "Constraint - dyn_skirtFront_r_01 (PP)"!(ComponentPose=Local To Component(LocalPose=Use cached pose 'upperBodyCache'()), Alpha=1, ConstraintWeights_0=0, ConstraintWeights_1:float=1) -> (Pose) {
    alphaBoolBlend {blendOption=Linear}
    boneToModify {boneName=dyn_skirtFront_r_01}
    constraintSetup [{targetBone {boneName=thigh_r}}, {targetBone {boneName=thigh_l}}]
    constraintWeights [1,1]
   }
   n38 = "Constraint - dyn_skirtFrontA_r_01 (PP)"!(ComponentPose=@n37, Alpha=0, ConstraintWeights_0=1, ConstraintWeights_1:float=0) -> (Pose) {
    alphaBoolBlend {blendOption=Linear}
    boneToModify {boneName=dyn_skirtFrontA_r_01}
    constraintSetup [{targetBone {boneName=thigh_r}}, {targetBone {boneName=thigh_l}}]
    constraintWeights [1,1]
   }
   n39 = RigidBody(ComponentPose=@n38, OverrideWorldGravity=Get Gravity Override()) {
    bOverrideWorldGravity=true
    overridePhysicsAsset=&/M_MED_AllKnowing_Skirt_FrontEnd_Physics
    simulationSpace=BaseBoneSpace
    alphaBoolBlend {blendOption=Linear}
    baseBoneRef {boneName=pelvis}
    componentAppliedLinearAccClamp=<6000 6000 6000>
    componentLinearAccScale=<0.05 0.05 0.05>
    componentLinearVelScale=<0.05 0.05 0.05>
   }
   n40 = "Transform (Modify) Bone - Bone: dyn_skirtFront_r_01"(ComponentPose=@n39, Alpha=1, Translation="0,0,0", Rotation="15,0,0", Scale="1,1,1") {
    rotationMode=BMM_Additive
    rotationSpace=BCS_BoneSpace
    alphaBoolBlend {blendOption=Linear}
    boneToModify {boneName=dyn_skirtFront_r_01}
    rotation {pitch=15}
   }
   n41 = "Transform (Modify) Bone - Bone: dyn_skirtFront_l_01"(ComponentPose=@n40, Alpha=1, Translation="0,0,0", Rotation="15,0,0", Scale="1,1,1") {
    rotationMode=BMM_Additive
    rotationSpace=BCS_BoneSpace
    alphaBoolBlend {blendOption=Linear}
    boneToModify {boneName=dyn_skirtFront_l_01}
    rotation {pitch=15}
   }
   n42 = Component To Local(ComponentPose=@n41)
   n43 = "Constraint - dyn_skirtFront_r_01 (PP)"(ComponentPose=Local To Component(LocalPose=Use cached pose 'upperBodyCache'()), Alpha=1, ConstraintWeights_0=0, ConstraintWeights_1=1) {
    alphaBoolBlend {blendOption=Linear}
    boneToModify {boneName=dyn_skirtFront_r_01}
    constraintSetup [{targetBone {boneName=thigh_r}}, {targetBone {boneName=thigh_l}}]
    constraintWeights [1,1]
   }
   n44 = "Constraint - dyn_skirtFrontA_r_01 (PP)"(ComponentPose=@n43, Alpha=1, ConstraintWeights_0=1, ConstraintWeights_1=0) {
    alphaBoolBlend {blendOption=Linear}
    boneToModify {boneName=dyn_skirtFrontA_r_01}
    constraintSetup [{targetBone {boneName=thigh_r}}, {targetBone {boneName=thigh_l}}]
    constraintWeights [1,1]
   }
   n45 = RigidBody(ComponentPose=@n44, OverrideWorldGravity=Get Gravity Override()) {
    bOverrideWorldGravity=true
    overridePhysicsAsset=&/M_MED_AllKnowing_Skirt_Physics
    simulationSpace=BaseBoneSpace
    alphaBoolBlend {blendOption=Linear}
    baseBoneRef {boneName=pelvis}
    componentAppliedLinearAccClamp=<6000 6000 6000>
    componentLinearAccScale=<0.05 0.05 0.05>
    componentLinearVelScale=<0.05 0.05 0.05>
   }
   n46 = Component To Local(ComponentPose=@n45)
   n47 = Get bIsInFrontEnd()
   n48 = RigidBody(ComponentPose=Local To Component(LocalPose=Use cached pose 'crouchPose'()), OverrideWorldGravity=Get Gravity Override()) {
    bOverrideWorldGravity=true
    overridePhysicsAsset=&/M_MED_AllKnowing_crouchingNew_Skirt_Physics
    simulationSpace=BaseBoneSpace
    alphaBoolBlend {blendOption=Linear}
    baseBoneRef {boneName=pelvis}
    componentAppliedLinearAccClamp=<6000 6000 6000>
    componentLinearAccScale=<0.05 0.05 0.05>
    componentLinearVelScale=<0.05 0.05 0.05>
   }
   n49 = "Constraint - dyn_skirtFront_r_01 (PP)"(ComponentPose=@n48, Alpha=1, ConstraintWeights_0=1, ConstraintWeights_1=0) {
    alphaBoolBlend {blendOption=Linear}
    boneToModify {boneName=dyn_skirtFront_r_01}
    constraintSetup [{targetBone {boneName=thigh_r}}, {targetBone {boneName=thigh_l}}]
    constraintWeights [1,1]
   }
   n50 = "Transform (Modify) Bone - Bone: dyn_skirtFront_r_01"(ComponentPose=@n49, Alpha=1, Translation="1.005,9.932,2.821", Rotation="12.039,3.765,44.594", Scale="1,1,1") {
    rotationMode=BMM_Additive
    rotationSpace=BCS_BoneSpace
    translationMode=BMM_Additive
    translationSpace=BCS_BoneSpace
    alphaBoolBlend {blendOption=Linear}
    boneToModify {boneName=dyn_skirtFront_r_01}
    rotation=<rot 12.039 3.765 44.594>
    translation=<1.005 9.932 2.821>
   }
   n51 = "Transform (Modify) Bone - Bone: dyn_skirtFront_r_02"!(ComponentPose=@n50, Alpha:float=1, Translation="0,0,0", Rotation="25,-10,0", Scale="1,1,1") -> (Pose) {
    rotationMode=BMM_Additive
    rotationSpace=BCS_BoneSpace
    translationMode=BMM_Additive
    translationSpace=BCS_BoneSpace
    alphaBoolBlend {blendOption=Linear}
    boneToModify {boneName=dyn_skirtFront_r_02}
    rotation {pitch=25, yaw=-10}
   }
   n52 = "Transform (Modify) Bone - Bone: dyn_skirtFront_r_03"!(ComponentPose=@n51, Alpha:float=1, Translation="0,0,0", Rotation="30,0,0", Scale="1,1,1") -> (Pose) {
    rotationMode=BMM_Additive
    rotationSpace=BCS_BoneSpace
    translationMode=BMM_Additive
    translationSpace=BCS_BoneSpace
    alphaBoolBlend {blendOption=Linear}
    boneToModify {boneName=dyn_skirtFront_r_03}
    rotation {pitch=30}
   }
   Save cached pose 'skirtPose'(Pose=Blend Poses by bool(
     BlendPose_0=Component To Local(ComponentPose=@n52)
     BlendPose_1=Blend Poses by bool(BlendPose_0=@n42, BlendPose_1=@n46, bActiveValue=@n47) {blendType=Linear}
     bActiveValue=Get crouch moving()
    ) {blendType=Linear}) {cacheName=skirtPose}
   Use cached pose 'upperBodyCache'()
   Get bIsCrouching()
  ]}
 EventGraph {nodes [
   n1 = Event BlueprintUpdateAnimation()
   > n2 = Set Delta Time X(Delta Time X=@n1.DeltaTimeX)
   > n3 = Sequence!(execute) -> (then_0, then_1, then_2:Exec, then_3:Exec, then_4:Exec, then_5:Exec, then_6:Exec)
   > n4 = Gravity Override State Function DEPRECATED(CharacterPartAnimInstance=Self-Reference(), GameTimeInSeconds=@n2, GravityOverrideParams=Get Gravity Override Params())
   > Set Gravity Override(Gravity Override=@n4)
   Set Is Wearing Backbling(execute=@n3.then_1, Is Wearing Backbling=CheckForBackbling(self=Self-Reference()))
   Set bodyMesh(execute=@n3.then_2, bodyMesh:Skeletal Mesh Component Object Reference=GetOwnerSkeletalMeshComponent()) -> (Output_Get:Skeletal Mesh Component Object Reference)
   Set thig_r_RotUp(
    execute=@n3.then_3
    thig_r_RotUp:float=GetAlphaFromReferenceTransformDelta(
     SkeletalMeshComponent=Get bodyMesh() -> (bodyMesh:Skeletal Mesh Component Object Reference)
     BoneName=thigh_r
     BaseName=pelvis
     InputRange=(X=0,Y=-85)
     TwistAxis="1,0,0"
    )
   ) -> (Output_Get:float)
   n5 = PlayerMovement(execute=@n3.then_4, AnimInstance=Self-Reference(), XDotProductCoefficient=1, YDotProductCoefficient=1, ZDotProductCoefficient=1)
   > n6 = Branch(Condition="Not Equal (Float)"(A=@n5.YMovement))
   > Set crouch moving(crouch moving:bool=false) -> (Output_Get:bool)
   > Set crouch still(crouch still:bool=false) -> (Output_Get:bool)
   n7 = Branch(execute=@n6.else, Condition=Get bIsCrouching())
   > Set crouch moving(crouch moving=true)
   > Set crouch still(crouch still=false)
   Set crouch moving(execute=@n7.else, crouch moving=false)
   > Set crouch still(crouch still=true)
   Set head Rot(
    execute=@n3.then_5
    head Rot:double=GetAlphaFromReferenceTransformDelta(SkeletalMeshComponent=Get bodyMesh(), BoneName=head, BaseName=neck_02, InputRange=(X=0,Y=-85), TwistAxis="1,0,0")
   ) -> (Output_Get:double)
   n8 = NOT Boolean(A=Get bRigidBodyEnabled())
   n9 = "integer > integer"(A=Get CurrentLOD(), B=1)
   Set isRigidBodyDisabled(execute=@n3.then_6, isRigidBodyDisabled:bool=AND Boolean(A=OR Boolean(A=@n8, B=@n9), B=Get bIsSkydiveFloating())) -> (Output_Get:bool)
  ]}
]
variables [
 bodyMesh {type=SkeletalMeshComponent}
 crouch moving {type=boolean}
 crouch still {type=boolean}
 Delta Time X {type=number}
 Gravity Override {type=Vector}
 Gravity Override Params {type=GravityOverrideParamsStruct}
 head Rot {type=number}
 Is Wearing Backbling {type=boolean}
 isRigidBodyDisabled {type=boolean}
 thig_r_RotUp {type=number}
]
