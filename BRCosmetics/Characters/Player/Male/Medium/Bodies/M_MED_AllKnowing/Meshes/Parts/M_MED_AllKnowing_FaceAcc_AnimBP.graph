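# Animation Blueprint for the AllKnowing male medium body's face accessory, adding idle, idle-break and eye motion over the attached parent's pose.
# The copied pose runs through DefaultSlot, then receives the idle state machine and eye animation as two additive layers.
# An EnemyPlayerEliminated input event briefly sets playerElimination after a short delay, triggering the non-looping Idle_Break.
# Post-evaluate clears the flag, and Idle_Break returns to Idle automatically without crossfade.
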AnimBlueprint /BRCosmetics/Characters/Player/Male/Medium/Bodies/M_MED_AllKnowing/Meshes/Parts/M_MED_AllKnowing_FaceAcc_AnimBP
target_skeleton=&/M_MED_AllKnowing_FaceAcc_Skeleton
graphs [
 AnimGraph {nodes [
   n1 = Slot 'DefaultSlot'(Source=Copy Pose From Mesh() {bUseAttachedParent=true})
   AnimGraphNode_StateMachine_0 = Idle_State() {is=title AnimGraphNode_StateMachine}
   Output Pose(Result=Apply Additive(
     Base=Apply Additive(Base=@n1, Additive=@AnimGraphNode_StateMachine_0) {alphaBoolBlend {blendOption=Linear}}
     Additive="Sequence Player (additive) 'AllKnowing_Eye_Anim_FaceAcc'"() {sequence=/BRCosmetics/Characters/Player/Male/Medium/Bodies/M_MED_AllKnowing/Meshes/Parts/Animation/AllKnowing_Eye_Anim_FaceAcc}
    ) {alphaBoolBlend {blendOption=Linear}}) {layerGroup=None}
  ]}
 Idle_State {
  entry=Idle
  path=AnimGraph.AnimGraphNode_StateMachine_0.Idle_State
  ids=id
  nodes [
   AnimStateNode_0 = Idle()
   AnimStateNode_1 = Idle_Break() {is=title AnimStateNode}
   AnimStateTransitionNode_0 = AnimStateTransitionNode()
   AnimStateTransitionNode_1 = AnimStateTransitionNode() {bAutomaticRuleBasedOnSequencePlayerInState=true, crossfadeDuration=0}
  ]
  links [
   {
    dst=Idle_Break
    rule=AnimGraph.AnimGraphNode_StateMachine_0.Idle_State.AnimStateTransitionNode_0.Transition
    src=Idle
    via=AnimStateTransitionNode_0
   }
   {
    dst=Idle
    rule=AnimGraph.AnimGraphNode_StateMachine_0.Idle_State.AnimStateTransitionNode_1.Transition
    src=Idle_Break
    via=AnimStateTransitionNode_1
   }
  ]
 }
 Idle {
  path=AnimGraph.AnimGraphNode_StateMachine_0.Idle_State.AnimStateNode_0.Idle
  nodes [Output Animation Pose(Result="Sequence Player (additive) 'AllKnowing_Idle_Anim_FaceAcc'"() {sequence=/BRCosmetics/Characters/Player/Male/Medium/Bodies/M_MED_AllKnowing/Meshes/Parts/Animation/AllKnowing_Idle_Anim_FaceAcc})]
 }
 Idle_Break {
  path=AnimGraph.AnimGraphNode_StateMachine_0.Idle_State.AnimStateNode_1.Idle_Break
  nodes [
   n1 = "Sequence Player (additive) 'AllKnowing_Idle_Break_Anim_FaceAcc'"() {
    bLoopAnimation=false
    sequence=/BRCosmetics/Characters/Player/Male/Medium/Bodies/M_MED_AllKnowing/Meshes/Parts/Animation/AllKnowing_Idle_Break_Anim_FaceAcc
   }
   Output Animation Pose(Result=@n1)
  ]
 }
 Transition {
  path=AnimGraph.AnimGraphNode_StateMachine_0.Idle_State.AnimStateTransitionNode_0.Transition
  nodes [Result(bCanEnterTransition=Get playerElimination() -> (playerElimination:bool))]
 }
 Transition {path=AnimGraph.AnimGraphNode_StateMachine_0.Idle_State.AnimStateTransitionNode_1.Transition, nodes [Result()]}
 EventGraph {nodes [
   n1 = Event BlueprintProcessAnimInputEvent()
   > Branch(Condition=IsNamed(self=@n1.AnimInputEvent, InEventName=aie.EnemyPlayerEliminated))
   > Delay(Duration=0.4)
   > Set playerElimination(playerElimination:bool=true) -> (Output_Get:bool)
   Event BlueprintPostEvaluateAnimation() {is=title K2Node_Event}
   > Set playerElimination(playerElimination=false)
   Event BlueprintUpdateAnimation()
  ]}
]
variables [playerElimination {type=boolean}]
